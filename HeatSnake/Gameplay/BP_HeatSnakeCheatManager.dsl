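; HeatSnake cheat manager with a client-only cheat, HeatSnakeSetAnimationScaling, that sets animation scaling on or off for HeatSnake characters.
; It prints an error on dedicated servers; otherwise it loops over all FortPlayerPawns and their meshes' child components.
; Skeletal mesh anim instances are cast to the FridayPloy HeatSnake Sharp skeleton AnimBlueprint to receive the boolean.
; On-screen messages report each change, or failure while no matching AnimBlueprint has been found.
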(fn HeatSnakeSetAnimationScaling
  (params (ScaleAnim :type Boolean))
  (bind _scaleanim (Variables|GetScaleAnim))
  (if (Networking|IsDedicatedServer)
    (Development|PrintString "This cheat does not run on server. Try rerunning without the \"cheat\" prefix to run on client." true true "(R=1.000000,G=0.000000,B=0.000000,A=1.000000)" 1e+01)
    (else
      (bind _outactors (Actor|GetAllActorsOfClass "/Script/FortniteGame.FortPlayerPawn"))
      (for _array_element _outactors
        (bind _array_element_2 (Utilities|Array|ForEachLoopwithBreak (Components|GetChildrenComponents (Variables|/Script/Engine.Character|GetMesh _array_element) true))
          (:LoopBody
            (bind _asskeletal_mesh_component (Utilities|Casting|CastToSkeletalMeshComponent _array_element_2))
            (bind _asfriday_ploy_heat_snake_sharp_skeleton_anim_blueprint (Utilities|Casting|CastToFridayPloy_HeatSnake_Sharp_Skeleton_AnimBlueprint (Components|SkeletalMesh|GetAnimInstance _asskeletal_mesh_component)))
            (Class|FridayPloyHeatSnakeSharpSkeletonAnimBlueprint|SetAnimationScaling _asfriday_ploy_heat_snake_sharp_skeleton_anim_blueprint _scaleanim)
            (LocalVariables|SetSuccess true)
            (Development|PrintString (Utilities|String|Append "Setting HeatSnake Animation Scaling for: " (Utilities|GetDisplayName _array_element) " to: " (Utilities|String|ToString(Boolean) _scaleanim)) true true "(R=0.000000,G=1.000000,B=0.000000,A=1.000000)" 5.0))
          (:Completed
            _))
        (if (not (LocalVariables|GetSuccess))
          (Development|PrintString "Cheat Command Failed. Could not find an appropriate HeatSnake ABP to apply cheat." true true "(R=1.000000,G=0.000000,B=0.000000,A=1.000000)" 1e+01))))))
